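# Sound cue for a hologram switching off, going by its name and path, in the fourth Monster Hunter quest of the Season 18 battle pass.
# It plays a reused Battle Lab auto-spawner hologram-off wave through a Modulator node left at default settings.
# It routes to the quest sound class at reduced volume, with overridden short-range attenuation and concurrency capped at one instance.
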SoundCue /BattlePassS18/Audio/MonsterHunter/Quest4/MH_Quest4_HologramOff_Cue
root=Modulator_0
nodes [
 {id=Modulator_0, children [WavePlayer_0]}
 {id=WavePlayer_0, properties {soundWaveAssetPtr=/Game/Sounds/BattleLab/AutoSpawner/AutoSpawner_HologramOff_01}}
]
properties {
 attenuationSettings=/BattlePassS18/Audio/S18_Narrative_Quest_Att
 bOverrideAttenuation=true
 bOverrideConcurrency=true
 maxDistance=768
 priority=0.9
 soundClassObject=&Fort_Quests
 volumeMultiplier=0.4
 attenuationOverrides {falloffDistance=512, attenuationShapeExtents {x=256}}
 concurrencyOverrides {maxCount=1}
}
