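# Sound cue for a hologram switch-on effect in Monster Hunter Quest 4 of the Season 18 Battle Pass content.
# It reuses the BattleLab AutoSpawner HologramOn wave, fed through a default Modulator for slight pitch and volume variation.
# It plays at half volume in the quests sound class, using the shared S18 narrative quest attenuation.
# A concurrency override limits it to two simultaneous instances.
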SoundCue /BattlePassS18/Audio/MonsterHunter/Quest4/MH_Quest4_HologramOn_Cue
root=Modulator_0
nodes [
 {id=Modulator_0, children [WavePlayer_1]}
 {id=WavePlayer_1, properties {soundWaveAssetPtr=/Game/Sounds/BattleLab/AutoSpawner/AutoSpawner_HologramOn_01}}
]
properties {
 attenuationSettings=/BattlePassS18/Audio/S18_Narrative_Quest_Att
 bOverrideConcurrency=true
 maxDistance=3512
 soundClassObject=&Fort_Quests
 volumeMultiplier=0.5
 attenuationOverrides {falloffDistance=512, attenuationShapeExtents {x=256}}
 concurrencyOverrides {maxCount=2}
}
